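# Editor utility widget for the quest editor that edits one objective filter, built on the C++ FortQuestObjectiveFilterEditor class.
# An expandable area holds a header with a numbered title, a detail view of the ObjectiveFilters value, and duplicate and remove buttons, above a properties detail view.
# The buttons call the parent's DuplicateObjectiveFilter and RemoveObjectiveFilter.
# When filter data is set, the title shows the filter index plus an offset.
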EditorUtilityWidgetBlueprint /Game/NeverCook/EditorUtilities/QuestTools/QuestEditor/QuestObjectiveEditors/ObjectiveFilterEditors/FortObjectiveFilterEditor
animations [
 In {
  binding_count=1
  length_seconds=0.5
  object=&self:In
  sequence {
   bindings [VerticalBox_Root {
     id=#1
     parent_id=null
     possessed_class=VerticalBox
     tracks [{
       type=MovieScene2DTransformTrack
       sections [{
         type=MovieScene2DTransformSection
         end_frame=10
         start_frame=0
         channels [
          Translation.X {type=MovieSceneScriptingFloatChannel, keys [[0,-100],[10,0]]}
          Translation.Y {type=MovieSceneScriptingFloatChannel}
          Angle {type=MovieSceneScriptingFloatChannel}
          Scale.X {type=MovieSceneScriptingFloatChannel}
          Scale.Y {type=MovieSceneScriptingFloatChannel}
          Shear.X {type=MovieSceneScriptingFloatChannel}
          Shear.Y {type=MovieSceneScriptingFloatChannel}
         ]
        }]
      }]
    }]
   display_rate {denominator=1, numerator=20}
   tick_resolution {denominator=1, numerator=60000}
   tracks []
  }
 }
 Out {
  binding_count=1
  length_seconds=0.25
  object=&self:Out
  sequence {
   bindings [VerticalBox_Root {
     id=#2
     parent_id=null
     possessed_class=VerticalBox
     tracks [{
       type=MovieSceneFloatTrack
       display_name=Render Opacity
       sections [{type=MovieSceneFloatSection, end_frame=5, start_frame=0, channels [None {type=MovieSceneScriptingFloatChannel, keys [[0,1],[5,0]]}]}]
      }]
    }]
   display_rate {denominator=1, numerator=20}
   tick_resolution {denominator=1, numerator=60000}
   tracks []
  }
 }
]
graphs [
 EventGraph {nodes [
   "On Clicked (Button_DuplicateFilter)"!() -> (OutputDelegate, then, Button:Common Button Base Object Reference) {is=title K2Node_ComponentBoundEvent}
   > DuplicateObjectiveFilter(self:Fort Quest Objective Filter Editor Object Reference)
   "On Clicked (Button_RemoveFilter)"!() -> (OutputDelegate, then, Button:Common Button Base Object Reference) {is=title K2Node_ComponentBoundEvent}
   > RemoveObjectiveFilter(self:Fort Quest Objective Filter Editor Object Reference)
   Event OnObjectiveFilterDataSet() {is=title K2Node_Event}
   n1 = Get ObjectiveFilterIndex() -> (ObjectiveFilterIndex:int=0)
   > Set Text(
    Text=Format Text!(Format="LOCTEXT(\"Objective Filter {0}\")", 0:int=int + int(A=@n1)) -> (Result)
    self=Get Text_StatHeader() -> (Text_StatHeader:Common Text Object Reference)
   )
  ]}
 OnStatDuplicated {nodes [OnStatDuplicated!() -> (then, ObjectiveIndex:int, StatIndex:int) {is=title K2Node_FunctionEntry}]}
 OnStatRemoved {nodes [OnStatRemoved!() -> (then, ObjectiveIndex:int, StatIndex:int, Widget:Widget Object Reference) {is=title K2Node_FunctionEntry}]}
]
variables [
 OnStatDuplicated {default=()}
 OnStatRemoved {default=()}
]
widgets {
 info {
  inheritedWidgetCount=3
  namedSlotCount=2
  parentClass=/Script/FortniteEditor.FortQuestObjectiveFilterEditor
  rootWidgetClass=&Border
  widgetCount=18
 }
 widgets [
  Border_1 {
   widgetClassPath=&Border
   widget=Border_1
   properties {
    background {
     bIsDynamicallyLoaded=false
     drawAs=Image
     imageType=NoImage
     mirroring=NoMirror
     resourceName=None
     resourceObject=null
     tiling=NoTile
     imageSize=<32 32>
     margin=<ltrb 0 0 0 0>
     outlineSettings {
      bUseBrushTransparency=false
      roundingType=HalfHeightRadius
      width=0
      color {colorUseRule=UseColor_Specified, specifiedColor=<rgba 0 0 0 0>}
      cornerRadii=<0 0 0 0>
     }
     tintColor {colorUseRule=UseColor_Specified, specifiedColor=<rgba 0.0103 0.0103 0.0103 1>}
     uVRegion {bIsValid=false, max=<0 0>, min=<0 0>}
    }
    padding=<ltrb 5 0 4 0>
   }
  }
  Border_0 {
   widgetClassPath=&Border
   parent=Border_1
   widget=Border_0
   properties {
    background {
     bIsDynamicallyLoaded=false
     drawAs=Image
     imageType=NoImage
     mirroring=NoMirror
     resourceName=None
     resourceObject=null
     tiling=NoTile
     imageSize=<32 32>
     margin=<ltrb 0 0 0 0>
     outlineSettings {
      bUseBrushTransparency=false
      roundingType=HalfHeightRadius
      width=0
      color {colorUseRule=UseColor_Specified, specifiedColor=<rgba 0 0 0 0>}
      cornerRadii=<0 0 0 0>
     }
     tintColor {colorUseRule=UseColor_Specified, specifiedColor=<rgba 0.0685 0.0685 0.0685 1>}
     uVRegion {bIsValid=false, max=<0 0>, min=<0 0>}
    }
    padding=<ltrb 0 0 0 2>
   }
   slot {class=BorderSlot, properties {padding=<ltrb 5 0 4 0>}}
  }
  VerticalBox_Root {
   widgetClassPath=&VerticalBox
   parent=Border_0
   widget=VerticalBox_Root
   slot {class=BorderSlot, properties {padding=<ltrb 0 0 0 2>}}
  }
  ExpandableArea_ObjectiveFilterArea {
   widgetClassPath=/Script/UMG.ExpandableArea
   bInherited=true
   bIsVariable=true
   parent=VerticalBox_Root
   widget=ExpandableArea_ObjectiveFilterArea
   properties {
    bIsExpanded=true
    borderBrush {
     bIsDynamicallyLoaded=false
     drawAs=RoundedBox
     imageType=NoImage
     mirroring=NoMirror
     resourceName=None
     resourceObject=null
     tiling=NoTile
     imageSize=<0 0>
     margin=<ltrb 0 0 0 0>
     outlineSettings {
      bUseBrushTransparency=false
      roundingType=FixedRadius
      width=0
      color {colorUseRule=UseColor_Specified, specifiedColor=<rgba 0 0 0 0>}
      cornerRadii=<4 4 4 4>
     }
     tintColor {colorUseRule=UseColor_Specified, specifiedColor=<rgba 0.0176 0.0176 0.0176 1>}
     uVRegion {bIsValid=false, max=<0 0>, min=<0 0>}
    }
    style {
     rolloutAnimationSeconds=0.1
     collapsedImage {
      bIsDynamicallyLoaded=false
      drawAs=Image
      imageType=Vector
      mirroring=NoMirror
      resourceName=../../../Engine/Content/Slate/Starship/Common/caret-right.svg
      resourceObject=null
      tiling=NoTile
      imageSize=<16 16>
      margin=<ltrb 0 0 0 0>
      outlineSettings {
       bUseBrushTransparency=false
       roundingType=HalfHeightRadius
       width=0
       color {colorUseRule=UseColor_Specified, specifiedColor=<rgba 0 0 0 0>}
       cornerRadii=<0 0 0 0>
      }
      tintColor {colorUseRule=UseColor_Specified, specifiedColor=<rgba 0.5271 0.5271 0.5271 1>}
      uVRegion {bIsValid=false, max=<0 0>, min=<0 0>}
     }
     expandedImage {
      bIsDynamicallyLoaded=false
      drawAs=Image
      imageType=Vector
      mirroring=NoMirror
      resourceName=../../../Engine/Content/Slate/Starship/Common/caret-down.svg
      resourceObject=null
      tiling=NoTile
      imageSize=<16 16>
      margin=<ltrb 0 0 0 0>
      outlineSettings {
       bUseBrushTransparency=false
       roundingType=HalfHeightRadius
       width=0
       color {colorUseRule=UseColor_Specified, specifiedColor=<rgba 0 0 0 0>}
       cornerRadii=<0 0 0 0>
      }
      tintColor {colorUseRule=UseColor_Specified, specifiedColor=<rgba 0.5271 0.5271 0.5271 1>}
      uVRegion {bIsValid=false, max=<0 0>, min=<0 0>}
     }
    }
   }
   slot {class=VerticalBoxSlot, properties {padding=<ltrb 20 0 20 0>, size {sizeRule=Fill, value=1}}}
  }
  VerticalBox_85 {widgetClassPath=&VerticalBox, namedSlotHost=ExpandableArea_ObjectiveFilterArea, widget=VerticalBox_85}
  HorizontalBox_1 {widgetClassPath=&HorizontalBox, parent=VerticalBox_85, widget=HorizontalBox_1, slot {class=VerticalBoxSlot}}
  Text_StatHeader {
   widgetClassPath=&CommonTextBlock
   bIsVariable=true
   parent=HorizontalBox_1
   widget=Text_StatHeader
   properties {
    text="Objective Filter {X}:"
    font {
     bForceMonospaced=false
     bMaterialIsStencil=false
     fontMaterial=null
     fontObject=&NotoSans
     letterSpacing=0
     monospacedWidth=1
     size=13
     skewAmount=0
     typefaceFontName=NotoSans-Bold
     outlineSettings {
      bApplyOutlineToDropShadows=false
      bMiteredCorners=false
      bSeparateFillAlpha=false
      outlineMaterial=null
      outlineSize=0
      outlineColor=<rgba 0 0 0 1>
     }
    }
    shadowOffset=<0 0>
   }
   slot {class=HorizontalBoxSlot, properties {verticalAlignment=VAlign_Center, padding=<ltrb 5 0 0 0>}}
  }
  ScaleBox_0 {
   widgetClassPath=&ScaleBox
   parent=HorizontalBox_1
   widget=ScaleBox_0
   properties {stretch=UserSpecified, userSpecifiedScale=0.25}
   slot {class=HorizontalBoxSlot, properties {horizontalAlignment=HAlign_Center, verticalAlignment=VAlign_Center}}
  }
  DetailView_Type {
   widgetClassPath=/Script/PropertiesDetailView
   bInherited=true
   bIsVariable=true
   parent=HorizontalBox_1
   widget=DetailView_Type
   properties {
    bShowPropertiesThatMatchGeneratedPropertyPaths=true
    propertiesToExclude [Objectives, ObjectiveStatHandle]
    propertiesToSkipPropertyPathGeneration [Objectives]
    propertyColumns [{
      fillCoefficient=1
      propertiesDisplayData [{
        bForceShowParentProperty=false
        bForceShowProperty=false
        bShowChildren=false
        bShowContainerProperty=false
        bShowName=false
        bShowPropertyButtons=false
        bShowValue=true
        containerObjectName=None
        propertyName=ObjectiveFilters
        propertyNameOverride=""
       }]
     }]
   }
   slot {class=HorizontalBoxSlot, properties {horizontalAlignment=HAlign_Center, verticalAlignment=VAlign_Center, padding=<ltrb 5 0 0 0>}}
  }
  Button_DuplicateFilter {
   widgetClassPath=/Game/NeverCook/EditorUtilities/QuestTools/QuestEditor/QuestEditorButtons/QuestEditorButton
   bIsVariable=true
   parent=HorizontalBox_1
   widget=Button_DuplicateFilter
   properties {bShowIcon=true, slateIconName=GenericCommands.Duplicate, text=Duplicate Filter, toolTipText=Make a new copy of the filter in the objective}
   slot {
    class=HorizontalBoxSlot
    properties {horizontalAlignment=HAlign_Right, verticalAlignment=VAlign_Center, padding=<ltrb 5 5 5 5>, size {sizeRule=Fill, value=1}}
   }
  }
  Button_RemoveFilter {
   widgetClassPath=&QuestEditorButton
   bIsVariable=true
   parent=HorizontalBox_1
   widget=Button_RemoveFilter
   properties {
    bShowIcon=true
    text=Remove Filter
    toolTipText=Removes filter from the objective
    iconBrush {
     bIsDynamicallyLoaded=false
     drawAs=Image
     imageType=NoImage
     mirroring=NoMirror
     resourceName=None
     resourceObject=/Game/UI/Foundation/Textures/Buttons/T-SlotIcon-Remove-32
     tiling=NoTile
     imageSize=<16 16>
     margin=<ltrb 0 0 0 0>
     outlineSettings {
      bUseBrushTransparency=false
      roundingType=HalfHeightRadius
      width=0
      color {colorUseRule=UseColor_Specified, specifiedColor=<rgba 0 0 0 0>}
      cornerRadii=<0 0 0 0>
     }
     tintColor {colorUseRule=UseColor_Specified, specifiedColor=<rgba 1 0.0017 0.0017 1>}
     uVRegion {bIsValid=false, max=<0 0>, min=<0 0>}
    }
   }
   slot {class=HorizontalBoxSlot, properties {horizontalAlignment=HAlign_Right, verticalAlignment=VAlign_Center, padding=<ltrb 5 5 5 5>}}
  }
  Image_Line {
   widgetClassPath=&Image
   bIsVariable=true
   parent=VerticalBox_85
   widget=Image_Line
   properties {brush {
     bIsDynamicallyLoaded=false
     drawAs=Image
     imageType=NoImage
     mirroring=NoMirror
     resourceName=None
     resourceObject=null
     tiling=NoTile
     imageSize=<1 1>
     margin=<ltrb 0 0 0 0>
     outlineSettings {
      bUseBrushTransparency=false
      roundingType=HalfHeightRadius
      width=0
      color {colorUseRule=UseColor_Specified, specifiedColor=<rgba 0 0 0 0>}
      cornerRadii=<0 0 0 0>
     }
     tintColor {colorUseRule=UseColor_Specified, specifiedColor=<rgba 0.0103 0.0103 0.0103 1>}
     uVRegion {bIsValid=false, max=<0 0>, min=<0 0>}
    }}
   slot {class=VerticalBoxSlot}
  }
  Border_Root {
   widgetClassPath=&Border
   bIsVariable=true
   namedSlotHost=ExpandableArea_ObjectiveFilterArea
   widget=Border_Root
   properties {brushColor=<rgba 0.0284 0.0284 0.0284 1>, padding=<ltrb 0 2 0 4>}
  }
  VerticalBox_48 {
   widgetClassPath=&VerticalBox
   parent=Border_Root
   widget=VerticalBox_48
   slot {class=BorderSlot, properties {padding=<ltrb 0 2 0 4>}}
  }
  EditorUtilityScrollBox_228 {
   widgetClassPath=&EditorUtilityScrollBox
   parent=VerticalBox_48
   widget=EditorUtilityScrollBox_228
   properties {orientation=Orient_Horizontal}
   slot {class=VerticalBoxSlot, properties {padding=<ltrb 4 2 4 0>}}
  }
  HorizontalBox_728 {
   widgetClassPath=&HorizontalBox
   parent=EditorUtilityScrollBox_228
   widget=HorizontalBox_728
   slot {class=ScrollBoxSlot, properties {size {sizeRule=Fill, value=1}}}
  }
  Border {
   widgetClassPath=&Border
   parent=HorizontalBox_728
   widget=Border
   properties {brushColor=<rgba 0.0103 0.0103 0.0103 1>, padding=<ltrb 0 0 0 0>}
   slot {class=HorizontalBoxSlot, properties {padding=<ltrb 2 0 0 2>, size {sizeRule=Fill, value=1}}}
  }
  DetailView_Properties {
   widgetClassPath=&PropertiesDetailView
   bInherited=true
   bIsVariable=true
   parent=Border
   widget=DetailView_Properties
   properties {
    bShowAllProperties=true
    bShowPropertiesThatMatchGeneratedPropertyPaths=true
    propertiesToExclude [ObjectiveStatHandle, ObjectiveFilters]
    propertiesToSkipPropertyPathGeneration [Objectives]
    propertyColumns [{fillCoefficient=1, propertiesDisplayData []}]
   }
   slot {class=BorderSlot, properties {padding=<ltrb 0 0 0 0>}}
  }
 ]
}
